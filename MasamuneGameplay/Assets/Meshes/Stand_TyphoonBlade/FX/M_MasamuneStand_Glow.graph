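# Translucent unlit glow material for the Masamune Typhoon Blade stand FX, drawn as soft camera-facing sprites.
# World position offset uses the Sprite function, with its center shifted from the object position along a local axis by a vertex-red lerp, plus a camera offset.
# A powered radial UV mask blends dark and light colors through shared emissive adjustments.
# The same mask, multiplied by a texture and intensity, drives depth-faded opacity.
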Material /MasamuneGameplay/Assets/Meshes/Stand_TyphoonBlade/FX/M_MasamuneStand_Glow
graph {
 properties {bAutomaticallySetUsageInEditor=false, blendMode=BLEND_Translucent, shadingModel=MSM_Unlit, usedShadingModels=Unlit}
 nodes [
  n1 = ObjectPositionWS()
  n2 = ScalarParameter() {defaultValue=30, parameterName=OffsetMax}
  n3 = ScalarParameter() {defaultValue=-30, parameterName=OffsetMin}
  n4 = VertexColor()
  n5 = Transform(None=Multiply(
    A=VectorParameter() {parameterName=OffSet, defaultValue=<rgba 0 1 0 1>}.RGB
    B=LinearInterpolate(A=@n2, B=@n3, Alpha=@n4.R) {constA=30, constB=-30}
   )) {transformSourceType=TRANSFORMSOURCE_Local}
  Add(
   A=MaterialFunctionCall!(
    Center Location=Add(A=@n1, B=@n5)
    0-1 UVs=TextureCoordinate()
    XY scale=ScalarParameter() {defaultValue=100, parameterName=Size}
    OverridePivot
    "Normals (Optional)"
   ) {materialFunction=&Sprite}.World position offset
   B=MaterialFunctionCall(Offset Amount=ScalarParameter() {defaultValue=20, parameterName=CameraOffset}) {materialFunction=/Engine/Functions/Engine_MaterialFunctions02/WorldPositionOffset/CameraOffset}.Camera Offset
  )
  n6 = Power(
   Base=SphereMask(A=TextureCoordinate(), B=Constant() {r=0.5}) {attenuationRadius=0.5, hardnessPercent=0}
   Exp=ScalarParameter() {defaultValue=3, parameterName=MaskPower}
  ) {constExponent=3}
  n7 = TextureSampleParameter2D() {parameterName=Default Tex, samplerType=SAMPLERTYPE_Masks, texture=&T_WhiteMask}
  DepthFade(
   Opacity=Multiply(A=Multiply(A=@n6, B=@n7.R), B=ScalarParameter() {defaultValue=0.5, parameterName=AlphaIntensity}) {constB=0.7}
   FadeDistance=ScalarParameter() {defaultValue=10, parameterName=FadeDistance}
  ) {fadeDistanceDefault=10}
  LocalPosition()
  MaterialFunctionCall(Emissive=LinearInterpolate(
    A=VectorParameter() {parameterName=Color Dark, defaultValue=<rgba 0.75 0.375 0.1875 1>}.RGB
    B=VectorParameter() {parameterName=Color Light, defaultValue=<rgba 1 0.9909 0.5 1>}.RGB
    Alpha=@n6
   )) {materialFunction=&MF_FX_EmissiveAdjustments}
 ]
}
